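; BGA_Alpaca is a timed goop volume that grants gameplay effects and an InBiome tag to pawns and vehicles inside it.
; Server overlap events stack loose tags as a ref count, so effects apply on first entry and clear on leaving the last volume.
; Pawns lacking a valid ASC are queued for timed retries.
; A looping timer polls physics objects, which lack server overlap events, caching them to add or remove the tag.
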(event EventOnMatchStarted
  (|CallParentFunction))

(event EventBeginPlay
  (bind _row_physicsobjectsoverlapcheckinterval (Variables|Default|GetRow_PhysicsObjectsOverlapCheckInterval))
  (bind _returnvalue (ScalableFloat|GetValueatLevel _row_physicsobjectsoverlapcheckinterval))
  ;[0.175 1 0.3813] Begin Play
  (|CallParentFunction)
  (Utilities|Time|SetTimerbyEvent (delegate Custom|KillSelf) (ScalableFloat|GetValueatLevel (Variables|Default|GetRow_MaxLifespan)))
  (Variables|Default|SetTagContainer_InBiome (GameplayTags|MakeGameplayTagContainerfromTag (Variables|Default|GetTag_InBiome)))
  ;[1 1 1] Looping semi-tick timer to check for overlapping physics objects, since they don't generate overlap events on server. Interval must be >0 to be active
  (if (> _returnvalue 0.0)
    (bind _returnvalue_2 (Utilities|Time|SetTimerbyEvent (EventDispatchers|CreateEvent :!event Custom|UpdateOverlappedPhysicsObjects) _returnvalue true))
    (Variables|Default|SetHandle_UpdateOverlappedPhysicsObjectsTimer _returnvalue_2)))

(event Custom|KillSelf
  ;[1 0.5976 0.0697] Kill Self
  (Actor|SetLifeSpan 0.1)
  (Development|PrintString "Kill Goop Volume" true true "(R=1.000000,G=0.000000,B=0.833006,A=1.000000)" 5.0)
  (Rendering|SetActorHiddenInGame true)
  ;[1 1 1] Looping semi-tick timer to check for overlapping physics objects, since they don't generate overlap events on server. Interval must be >0 to be active
  (Collision|SetCollisionEnabled (Variables|Default|GetCollisionMesh) "NoCollision")
  (Utilities|Time|ClearandInvalidateTimerbyHandle (Variables|Default|GetHandle_UpdateOverlappedPhysicsObjectsTimer)))

(event Custom|TryValidatePawnASC
  ;[0.3419 0.3231 0.9301] Try Validate Pawn ASC
  (Utilities|Time|ClearandInvalidateTimerbyHandle (Variables|Default|GetHandle_TryValidatePlayerASCTimer))
  (if (> (Utilities|Array|Length (Variables|Default|GetPendingPawnsWithInvalidASC)) 0)
    (for _array_element (Variables|Default|GetPendingPawnsWithInvalidASC)
      (Utilities|Array|AddUnique (Variables|Default|GetPawnsWithInvalidASC) _array_element))
    (Utilities|Array|Clear (Variables|Default|GetPendingPawnsWithInvalidASC))
    (else
      _))
  (bind (_arrayindex _arrayelement) (Utilities|Array|ReverseforEachLoop (Variables|Default|GetPawnsWithInvalidASC))
    (:LoopBody
      (Utilities|IsValid _arrayelement
        (:Is%20Valid
          (Utilities|IsValid (Ability|GetFortAbilitySystemComponent _arrayelement)
            (:Is%20Valid
              (CallFunction|AddPawnTagsandEffects _arrayelement)
              (Utilities|Array|RemoveIndex (Variables|Default|GetPawnsWithInvalidASC) _arrayindex))))
        (:Is%20Not%20Valid
          (Utilities|Array|RemoveIndex (Variables|Default|GetPawnsWithInvalidASC) _arrayindex))))
    (:Completed
      (if (> (Utilities|Array|Length (Variables|Default|GetPawnsWithInvalidASC)) 0)
        (bind _returnvalue (Utilities|Time|SetTimerbyEvent (EventDispatchers|CreateEvent :!event Custom|TryValidatePawnASC) (Variables|Default|GetDelayBeforeTryValidatePlayerASC)))
        (Variables|Default|SetHandle_TryValidatePlayerASCTimer _returnvalue)))))

(event Custom|AddPawnTagsAndEffects
  (params (FortPawn :type Fort%20Pawn%20Object%20Reference))
  ;[0.07733 1 0] TODO:l format cleanup
  ;;[0.9011 0.1274 1] On Begin Overlap, ref count being in a goop
  (Development|PrintString (Utilities|String|Append "Goop applied tags/effects to " (Utilities|GetDisplayName FortPawn)) true true "(R=0.000000,G=1.000000,B=0.063113,A=1.000000)" 5.0)
  (if (GameplayTags|HasMatchingGameplayTag FortPawn (Variables|Default|GetTag_InBiome))
    (Development|PrintString "In additional Goop")
    (else
      (CallFunction|ApplyPawnEffects FortPawn)))
  (Ability|GameplayEffect|AddLooseGameplayTags FortPawn (Variables|Default|GetTagContainer_InBiome)))

(event OnComponentBeginOverlap(CollisionMesh)
  (params (OverlappedComponent :type Primitive%20Component%20Object%20Reference) (OtherActor :type Actor%20Object%20Reference) (OtherComp :type Primitive%20Component%20Object%20Reference) (OtherBodyIndex :type Integer) (bFromSweep :type Boolean) (SweepResult :type Hit%20Result%20Structure%20%28by%20ref%29))
  ;[0.07733 1 0] TODO:l format cleanup
  ;;[0.9011 0.1274 1] On Begin Overlap, ref count being in a goop
  (Networking|SwitchHasAuthority
    (:Authority
      (bind _asfort_pawn (Utilities|Casting|CastToFortPawn OtherActor)
        (:then
          (Utilities|IsValid (Ability|GetFortAbilitySystemComponent _asfort_pawn)
            (:Is%20Valid
              (CallFunction|AddPawnTagsandEffects _asfort_pawn))
            (:Is%20Not%20Valid
              ;[0.3419 0.3231 0.9301] Store Pawns with invalid ASC to try validating ASC later - fixes issue of pawn spawning inside the volume and trying to apply effects before the ASC is valid
              (Development|PrintString (Utilities|String|Append "Goop detected invalid ASC on " (Utilities|GetDisplayName _asfort_pawn)) true true "(R=1.000000,G=0.037158,B=0.000000,A=1.000000)" 5.0)
              (if (Utilities|Time|DoesTimerExistbyHandle (Variables|Default|GetHandle_TryValidatePlayerASCTimer))
                (Utilities|Array|AddUnique (Variables|Default|GetPendingPawnsWithInvalidASC) _asfort_pawn)
                (else
                  (Utilities|Array|AddUnique (Variables|Default|GetPawnsWithInvalidASC) _asfort_pawn)
                  (bind _returnvalue (Utilities|Time|SetTimerbyEvent (EventDispatchers|CreateEvent :!event Custom|TryValidatePawnASC) (Variables|Default|GetDelayBeforeTryValidatePlayerASC)))
                  (Variables|Default|SetHandle_TryValidatePlayerASCTimer _returnvalue))))))
        (:CastFailed
          (bind _asfort_vehicle_interface (Utilities|Casting|CastToFortVehicleInterface OtherActor))
          (Utilities|IsValid (Vehicle|GetVehicleAbilitySystemComponent _asfort_vehicle_interface)
            (:Is%20Valid
              (bind _returnvalue_2 (GameplayTags|HasMatchingGameplayTag(Message) OtherActor (Variables|Default|GetTag_InBiome)))
              (if _returnvalue_2
                (Development|PrintString "Vehicle In additional Goop")
                (else
                  (Development|PrintString "Vehicle In First Goop")
                  (CallFunction|ApplyVehicleEffects _asfort_vehicle_interface)))
              (Ability|GameplayEffect|AddLooseGameplayTags OtherActor (Variables|Default|GetTagContainer_InBiome)))))))))

(event Custom|TryGetVehicleAbilitySystemComponent)

(event OnComponentEndOverlap(CollisionMesh)
  (params (OverlappedComponent :type Primitive%20Component%20Object%20Reference) (OtherActor :type Actor%20Object%20Reference) (OtherComp :type Primitive%20Component%20Object%20Reference) (OtherBodyIndex :type Integer))
  ;[0.07733 1 0] TODO:l format cleanup
  ;;[1 1 1] On End Overlap, remove low grav if outside all goop
  (Networking|SwitchHasAuthority
    (:Authority
      (bind _asfort_pawn (Utilities|Casting|CastToFortPawn OtherActor)
        (:then
          (Ability|GameplayEffect|RemoveLooseGameplayTags _asfort_pawn (Variables|Default|GetTagContainer_InBiome))
          (if (GameplayTags|HasAnyMatchingGameplayTags _asfort_pawn (Variables|Default|GetTagContainer_InBiome))
            (Development|PrintString "Left A Goop, Still inside another")
            (else
              (CallFunction|RemovePawnEffects _asfort_pawn))))
        (:CastFailed
          (bind _asfort_vehicle_interface (Utilities|Casting|CastToFortVehicleInterface OtherActor))
          (Ability|GameplayEffect|RemoveLooseGameplayTags OtherActor (Variables|Default|GetTagContainer_InBiome))
          (bind _returnvalue (GameplayTags|HasAnyMatchingGameplayTags(Message) OtherActor (Variables|Default|GetTagContainer_InBiome)))
          (if _returnvalue
            (Development|PrintString "Vehicle Left A Goop, Still inside another")
            (else
              (CallFunction|RemoveVehicleEffects _asfort_vehicle_interface))))))))

(event Custom|UpdateOverlappedPhysicsObjects
  ;[1 1 1] Update Overlapped Physics Objects
  (Networking|SwitchHasAuthority
    (:Authority
      (bind _outactors (Collision|ComponentOverlapActors (Variables|Default|GetCollisionMesh) (Transformation|GetWorldTransform (Variables|Default|GetCollisionMesh)) (Variables|Default|GetObjectType_PhysicsObjects)))
      (bind _setcachedoverlappedphysicsactors (CallFunction|UpdateCachedPhysicsActors _outactors (Variables|Default|GetCachedOverlappedPhysicsObjects) (Variables|Default|GetTag_InBiome)))
      (Variables|Default|SetCachedOverlappedPhysicsObjects _setcachedoverlappedphysicsactors))))

(fn ApplyPawnEffects
  (params (Pawn :type Fort%20Pawn%20Object%20Reference))
  (GameplayEffects|FortApplyGameplayEffectToTarget (Ability|GetFortAbilitySystemComponent (Variables|GetPawn)) (Variables|Default|GetGE_GrantedInBGA_Pawns) (Ability|GetFortAbilitySystemComponent (Variables|GetPawn)) self self self 1.0 (Utilities|Struct|MakeGameplayEffectContextHandle)))

(fn ApplyVehicleEffects
  (params (VehicleInterface :type Fort%20Vehicle%20Interface%20Interface))
  (GameplayEffects|ApplyGameplayEffectToTarget (Vehicle|GetVehicleAbilitySystemComponent VehicleInterface) (Variables|Default|GetGE_GrantedInBGA_Vehicles) (Vehicle|GetVehicleAbilitySystemComponent VehicleInterface) 0.0 (Utilities|Struct|MakeGameplayEffectContextHandle)))

(fn RemovePawnEffects
  (params (Pawn :type Fort%20Pawn%20Object%20Reference))
  (GameplayEffects|RemoveActiveGameplayEffectbySourceEffect (Ability|GetFortAbilitySystemComponent (Variables|GetPawn)) (Variables|Default|GetGE_GrantedInBGA_Pawns) (Ability|GetFortAbilitySystemComponent (Variables|GetPawn)) -1))

(fn RemoveVehicleEffects
  (params (VehicleInterface :type Fort%20Vehicle%20Interface%20Interface))
  (GameplayEffects|RemoveActiveGameplayEffectbySourceEffect (Vehicle|GetVehicleAbilitySystemComponent VehicleInterface) (Variables|Default|GetGE_GrantedInBGA_Vehicles) (Vehicle|GetVehicleAbilitySystemComponent VehicleInterface) -1))

(fn UpdateCachedPhysicsActors
  (params (PhysicsActorsToProcess :type Array%20of%20Actor%20Object%20References) (CachedOverlappedPhysicsActors :type Array%20of%20Actor%20Object%20References) (TagCachedOverlappedPhysicsActors :type Gameplay%20Tag%20Structure))
  (Utilities|Array|Clear (LocalVariables|GetNewlyOverlappedPhysicsActors))
  (Utilities|Array|Clear (LocalVariables|GetOverlappedPhysicsActorsToCache))
  (for _array_element (Variables|GetPhysicsActorsToProcess)
    (bind _returnvalue (GameplayTags|HasMatchingGameplayTag(Message) _array_element (Variables|GetTagCachedOverlappedPhysicsActors))) ; Unlike pawns and vehicles, physics objects will only get 1 copy of this loose gameplay tag applied, because checks for them are on a ticking basis, not on overlap events.
    (if _returnvalue
      (if (Utilities|Array|ContainsItem (Variables|GetCachedOverlappedPhysicsActors) _array_element)
        (Utilities|Array|AddUnique (LocalVariables|GetNewlyOverlappedPhysicsActors) _array_element))
      (else
        (Utilities|Array|AddUnique (LocalVariables|GetNewlyOverlappedPhysicsActors) _array_element)
        (Ability|GameplayEffect|AddLooseGameplayTags _array_element (GameplayTags|MakeGameplayTagContainerfromTag (Variables|GetTagCachedOverlappedPhysicsActors))))))
  (for _array_element_2 (Variables|GetCachedOverlappedPhysicsActors)
    (if (Utilities|Array|ContainsItem (LocalVariables|GetNewlyOverlappedPhysicsActors) _array_element_2)
      (Utilities|Array|AddUnique (LocalVariables|GetOverlappedPhysicsActorsToCache) _array_element_2)
      (else
        (Ability|GameplayEffect|RemoveLooseGameplayTags _array_element_2 (GameplayTags|MakeGameplayTagContainerfromTag (Variables|GetTagCachedOverlappedPhysicsActors))))))
  (for _array_element_3 (LocalVariables|GetNewlyOverlappedPhysicsActors)
    (Utilities|Array|AddUnique (LocalVariables|GetOverlappedPhysicsActorsToCache) _array_element_3))
  (return (LocalVariables|GetOverlappedPhysicsActorsToCache)))

(fn ConstructionScript)
